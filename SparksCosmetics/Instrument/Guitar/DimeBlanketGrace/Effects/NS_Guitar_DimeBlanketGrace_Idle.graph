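# Idle sparkle effect for the DimeBlanketGrace guitar cosmetic, scattering short-lived flare sprites over the body, neck and headstock.
# Sparkles picks spawn points by weighted random section blends, then kills particles inside two region masks and a distance-tested soundhole.
# Quality level, time of day, LOD distance and music-preview values modulate the look, while User.Variant blends flares from warm to cool.
# A disabled Debug emitter color-codes that exclusion test.
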NiagaraSystem /SparksCosmetics/Instrument/Guitar/DimeBlanketGrace/Effects/NS_Guitar_DimeBlanketGrace_Idle
dependencies {
 usedDataInterfaces [&NiagaraDataInterfaceArrayFloat, &NiagaraDataInterfaceCurve, &NiagaraDataInterfaceColorCurve]
 usedDynamicInputs [
  &Multiply_Vector2D_ByFloat
  &MakeVector2D
  &ColorFromCurve
 ]
 usedModules [
  &SystemState
  &self:SystemSpawnScript.SystemScriptSource.SystemScriptGraph.NiagaraNodeAssignment_4.SetVariables_#1
  &self:SystemSpawnScript.SystemScriptSource.SystemScriptGraph.NiagaraNodeAssignment_6.SetVariables_#2
 ]
}
emitters {
 Debug {
  bEnabled=false
  emitterUpdateScript [
   &EmitterState()
   &SpawnRate(SpawnRate=3000, Use Spawn Probability=true)
  ]
  particleSpawnScript [
   /SetVariables(
    Transient.BodySectionA:Vector3f=&Add_Vector(
     A=<0 -3.1 10>
     B=/Niagara/DynamicInputs/Math/Clamp_Vector(
      Vector:Vector3f=&Multiply_Vector(A=<1 1 0.6>, B=&RandomVector(Vector Scale=12.5))
      Min:Vector3f=<-13 -4 -8>
      Max:Vector3f=<13 4 8>
     )
    )
    Transient.BodySectionB:Vector3f=&Add_Vector(
     A=<0 -3.1 -11.5>
     B=&Clamp_Vector(
      Vector:Vector3f=&Multiply_Vector(A=<1 1 0.75>, B=&RandomVector(Vector Scale=16.75))
      Min:Vector3f=<-18 -4 -14>
      Max:Vector3f=<18 4 14>
     )
    )
   ) {isSetParameters=true, module=SetVariables_#3}
   &InitializeParticle(
    AdjustAlpha=false
    AdjustHue=true
    AdjustSaturation=true
    AdjustValue=true
    Color=<rgba 2 2 2 1>
    Color Mode=::NewEnumerator3
    Hue Shift Range:Vector2f=<-0.05 0.05>
    Lifetime Max=1
    Lifetime Min=0.5
    Lifetime Mode=::NewEnumerator1
    Position Offset=&Lerp_Vector(
     A=$Transient.BodySectionA
     B=$Transient.BodySectionB
     Alpha=&UniformAOrBFloat(A=0, Distribution Weight=0.6, Evaluation Type=::NewEnumerator0, Recalculate Random Each Loop=true, Override Seed=false)
    )
    Sprite Size Mode=::NewEnumerator3
    Uniform Sprite Size=&Multiply_Float(
     A=&RandomRangeFloat(Minimum=12, Maximum=18)
     B=&SelectFloatFromArray(Direct Array Index=$Engine.QualityLevel, Float Selection Array=DataInterface() {propertyValues {FloatData [0.5,0.6,0.8,1,1]}})
    )
    UsePositionOffset=true
    Write Material Random=true
   )
   &RegionMask(
    Offset:Vector3f=<0 -3.1 3.5>
    Offset Space:ENiagaraCoordinateSpace=::Local
    Region Dimensions:Vector3f=<37 10 25>
    Region Falloff=0
    Region Type:ENiagaraDebugDrawShape=::NewEnumerator1
    UseOffset=true
    UseRotation=false
   )
   &RegionMask(
    Offset:Vector3f=<0 -6 0>
    Offset Space:ENiagaraCoordinateSpace=::Local
    Region Dimensions:Vector3f=<38 5 40>
    Region Falloff=0
    Region Type:ENiagaraDebugDrawShape=::NewEnumerator1
    UseOffset=true
    UseRotation=false
   ) {module=RegionMask001}
   /SetVariables(Transient.bIsInSoundhole:NiagaraBool=&Boolean_OrOperation(
     Bool1=$Particles.RegionMask.IsInsideMask
     Bool2=&SetBoolByFloatComparison(
      A=&VectorLength(Vector=&Add_Vector(A=Linked() {linkedVariable {name=Particles.Position, type=NiagaraPosition}}, B=<0 3.1 -3.5>))
      B=8
      Comparison Type=::NewEnumerator6
     )
    )) {enabled=false, isSetParameters=true, module=SetVariables_#4}
   /SetVariables(Transient.bIsInSoundhole:NiagaraBool=&SetBoolByFloatComparison(
     A=&VectorLength(Vector=&Add_Vector(A=Linked() {linkedVariable {name=Particles.Position, type=NiagaraPosition}}, B=<0 7.25 -3.5>))
     B=10
     Comparison Type=::NewEnumerator6
    )) {enabled=false, isSetParameters=true}
   /SetVariables(Transient.bIsInSoundhole:NiagaraBool=&Boolean_OrOperation(
     Bool1=&Boolean_OrOperation(Bool1=$Particles.RegionMask.IsInsideMask, Bool2=$Particles.RegionMask001.IsInsideMask)
     Bool2=&SetBoolByFloatComparison(
      A=&VectorLength(Vector=&Add_Vector(A=Linked() {linkedVariable {name=Particles.Position, type=NiagaraPosition}}, B=<0 3.1 -3.5>))
      B=8
      Comparison Type=::NewEnumerator6
     )
    )) {isSetParameters=true}
   &KillParticles(Kill Particles=$Transient.bIsInSoundhole) {enabled=false}
   &CameraOffset(Camera Offset Amount=1)
   &Color(Color=&MakeCustomLinearColorFromBool(Bool=$Transient.bIsInSoundhole, True Color=<rgba 0 0 1 10>, False Color=<rgba 1 0 0 10>))
  ]
  particleUpdateScript [
   &ParticleState()
   &ScaleColor(
    Scale Alpha=&Multiply_Float(
     A=&FloatFromCurve(
      FloatCurve=DataInterface() {propertyValues {Curve {keys [
          {arriveTangent=6.5707, interpMode=RCIM_Cubic, leaveTangent=6.5707, tangentMode=RCTM_User, time=0, value=0}
          {interpMode=RCIM_Cubic, tangentMode=RCTM_User, time=0.3, value=1}
          {tangentMode=RCTM_User, time=1, value=0}
         ]}}}
      Scale Curve=&RandomRangeFloat(Minimum=0.1, Maximum=0.5)
     )
     B=&MakeFloatFromLinearColor(LinearColor=$System.TODColorMultiplier, Channel:ENiagaraLinearColor_Channels=::NewEnumerator3)
    )
    Scale RGB=&Multiply_Vector(
     A=&VectorFromFloat(Value=&Multiply_Float(
       A=&RandomRangeFloat(Minimum=0.75, Maximum=1)
       B=&SelectFloatFromArray(Direct Array Index=$Engine.QualityLevel, Float Selection Array=DataInterface() {propertyValues {FloatData [10,5,1,1,1]}})
      ))
     B=&MakeVectorFromLinearColorRGB(Color=$System.TODColorMultiplier)
    )
    ScaleA=true
    ScaleRGB=true
   ) {module=ScaleColor001}
   &ScaleSpriteSize(
    Uniform Curve Scale=&MakeCustomFloatFromBool(
     Bool=$User.bIsFrontEndPreview
     True Float=1
     False Float=&FloatFromCurve(
      FloatCurve=DataInterface() {propertyValues {Curve {keys [<at 0 1>,<at 1000 2>]}}}
      CurveIndex=$Engine.Owner.LODDistance
      Scale Curve=&FNTimeOfDayScalar(Night Scale=1.25, Morning Scale=1, Evening Scale=1)
     )
    )
    Uniform Curve Sprite Scale=DataInterface() {propertyValues {Curve {keys [
        {interpMode=RCIM_Cubic, tangentMode=RCTM_User, time=0, value=0}
        {interpMode=RCIM_Cubic, tangentMode=RCTM_User, time=0.5, value=1}
        {tangentMode=RCTM_User, time=1, value=0}
       ]}}}
   )
   /SetVariables(
    Particles.CenterFlareColor:LinearColor=&MultiplyLinearColor(
     Linear Color=$Particles.Color
     Scale Factor=&MakeVector4_FromLinearColor(LinearColor=&LerpLinearColors(StartColor=<rgba 10 9.7549 6.1391 1>, EndColor=<rgba 6.0696 7.8037 10 1>, LerpFactor=$User.Variant))
    )
    Particles.CenterFlareSize:Vector2f=&Multiply_Vector2(A=$Particles.SpriteSize, B=&Vector2DFromFloat(Value=0.5))
    Particles.CoronaFlareColor:LinearColor=&MultiplyLinearColor(
     Linear Color=$Particles.Color
     Scale Factor=&MakeVector4_FromLinearColor(LinearColor=&LerpLinearColors(StartColor=<rgba 0.15 0.15 0.15 1>, EndColor=<rgba 0.0729 0.0729 0.0729 1>, LerpFactor=$User.Variant))
    )
    Particles.CoronaFlareSize:Vector2f=&Multiply_Vector2(A=$Particles.Initial.SpriteSize, B=&Vector2DFromFloat(Value=0.75))
    Particles.DebugScale:Vector2f=&Multiply_Vector2(A=$Particles.SpriteSize, B=&Vector2DFromFloat(Value=0.1))
   ) {isSetParameters=true, module=SetVariables_#5}
  ]
  rendererClasses [&NiagaraSpriteRendererProperties]
  renderers [
   {rendererClass=&NiagaraSpriteRendererProperties}
   {rendererClass=&NiagaraSpriteRendererProperties, rendererIndex=1}
   {rendererClass=&NiagaraSpriteRendererProperties, rendererIndex=2}
  ]
 }
 Sparkles {
  emitterUpdateScript [
   &EmitterState()
   &SpawnRate(
    Spawn Probability=&MakeCustomFloatFromBool(Bool=&SetBoolByFloatComparison(A=$System.Music_Reactive_FX_Preview, B=0), True Float=1, False Float=0.125)
    SpawnRate=4
    Use Spawn Probability=true
   )
  ]
  particleSpawnScript [
   /SetVariables(
    Transient.BodySectionA:Vector3f=&Add_Vector(
     A=<0 -3.25 10>
     B=&Clamp_Vector(
      Vector:Vector3f=&Multiply_Vector(A=<1 1 0.6>, B=&RandomVector(Vector Scale=12.5))
      Min:Vector3f=<-13 -3.5 -8>
      Max:Vector3f=<13 3.5 8>
     )
    )
    Transient.BodySectionB:Vector3f=&Add_Vector(
     A=<0 -3.25 -11.5>
     B=&Clamp_Vector(
      Vector:Vector3f=&Multiply_Vector(A=<1 1 0.75>, B=&RandomVector(Vector Scale=16.75))
      Min:Vector3f=<-18 -3.5 -14>
      Max:Vector3f=<18 3.5 14>
     )
    )
    Transient.HeadSection:Vector3f=&Add_Vector(
     A=<0 0 66>
     B=/Niagara/DynamicInputs/Vectors/RotateVector(
      Vector To Rotate:Vector3f=&RandomRangeVector(Minimum=<-4 -1 -7>, Maximum=<4 1 7>)
      Yaw:NiagaraFloat=0
      Pitch:NiagaraFloat=0
      Roll:NiagaraFloat=-10
      Rotation Rate:NiagaraFloat=1
     )
    )
    Transient.NeckSection:Vector3f=&Add_Vector(
     A=<0 -1 20>
     B=&RotateVector(
      Vector To Rotate:Vector3f=&RandomRangeVector(Minimum=<-2.5 -0.5 0>, Maximum=<2.5 0.5 50>)
      Yaw:NiagaraFloat=0
      Pitch:NiagaraFloat=0
      Roll:NiagaraFloat=1
      Rotation Rate:NiagaraFloat=1
     )
    )
   ) {isSetParameters=true, module=SetVariables_#3}
   &InitializeParticle(
    AdjustAlpha=false
    AdjustHue=true
    AdjustSaturation=true
    AdjustValue=true
    Color=<rgba 2 2 2 1>
    Color Mode=::NewEnumerator3
    Hue Shift Range:Vector2f=<-0.05 0.05>
    Lifetime Max=1
    Lifetime Min=0.5
    Lifetime Mode=::NewEnumerator1
    Position Offset=&Lerp_Vector(
     A=&Lerp_Vector(
      A=$Transient.HeadSection
      B=$Transient.NeckSection
      Alpha=&UniformAOrBFloat(A=0, Distribution Weight=0.6, Evaluation Type=::NewEnumerator0, Recalculate Random Each Loop=true, Override Seed=false)
     )
     B=&Lerp_Vector(
      A=$Transient.BodySectionA
      B=$Transient.BodySectionB
      Alpha=&UniformAOrBFloat(A=0, Distribution Weight=0.6, Evaluation Type=::NewEnumerator0, Recalculate Random Each Loop=true, Override Seed=false)
     )
     Alpha=&UniformAOrBFloat(A=0, Distribution Weight=0.85, Evaluation Type=::NewEnumerator0, Recalculate Random Each Loop=true, Override Seed=false)
    )
    Sprite Size Mode=::NewEnumerator3
    Uniform Sprite Size=&Multiply_Float(
     A=&RandomRangeFloat(Minimum=12, Maximum=18)
     B=&SelectFloatFromArray(Direct Array Index=$Engine.QualityLevel, Float Selection Array=DataInterface() {propertyValues {FloatData [0.5,0.6,0.8,1,1]}})
    )
    UsePositionOffset=true
    Write Material Random=true
   )
   &RegionMask(
    Offset:Vector3f=<0 -3.1 3.5>
    Offset Space:ENiagaraCoordinateSpace=::Local
    Region Dimensions:Vector3f=<37 10 25>
    Region Falloff=0
    Region Type:ENiagaraDebugDrawShape=::NewEnumerator1
    UseOffset=true
    UseRotation=false
   )
   &RegionMask(
    Offset:Vector3f=<0 -6 0>
    Offset Space:ENiagaraCoordinateSpace=::Local
    Region Dimensions:Vector3f=<38 5 40>
    Region Falloff=0
    Region Type:ENiagaraDebugDrawShape=::NewEnumerator1
    UseOffset=true
    UseRotation=false
   ) {module=RegionMask001}
   /SetVariables(Transient.bIsInSoundhole:NiagaraBool=&Boolean_OrOperation(
     Bool1=&Boolean_OrOperation(Bool1=$Particles.RegionMask.IsInsideMask, Bool2=$Particles.RegionMask001.IsInsideMask)
     Bool2=&SetBoolByFloatComparison(
      A=&VectorLength(Vector=&Add_Vector(A=Linked() {linkedVariable {name=Particles.Position, type=NiagaraPosition}}, B=<0 3.1 -3.5>))
      B=8
      Comparison Type=::NewEnumerator6
     )
    )) {isSetParameters=true, module=SetVariables_#4}
   &KillParticles(Kill Particles=$Transient.bIsInSoundhole)
   &CameraOffset(Camera Offset Amount=&FN_FrontEndFloatModifier(
     Gameplay=&MakeCustomFloatFromBool(Bool=$User.bIsFrontEndPreview, True Float=1, False Float=2)
     Front End=1
    ))
  ]
  particleUpdateScript [
   &ParticleState()
   /SetVariables(Transient.BeatReactiveMult:NiagaraFloat=&MakeCustomFloatFromBool(
     Bool=$User.MusicReactive
     True Float=1
     False Float=&MakeCustomFloatFromBool(
      Bool=&SetBoolByFloatComparison(A=$System.Music_Reactive_FX_Preview, B=0)
      True Float=&Lerp_Float(
       A=&FN_FrontEndFloatModifier(
        Gameplay=&MakeCustomFloatFromBool(Bool=$User.bIsFrontEndPreview, True Float=1, False Float=1.4)
        Front End=1
       )
       B=&FN_FrontEndFloatModifier(
        Gameplay=&MakeCustomFloatFromBool(Bool=$User.bIsFrontEndPreview, True Float=1.75, False Float=2)
        Front End=2
       )
       Alpha=$System.Music_Reactive_FX_Preview
      )
      False Float=1
     )
    )) {isSetParameters=true}
   &ScaleColor(
    Scale Alpha=&Multiply_Float(
     A=$Transient.BeatReactiveMult
     B=&Multiply_Float(
      A=&FloatFromCurve(
       FloatCurve=DataInterface() {propertyValues {Curve {keys [
           {arriveTangent=6.5707, interpMode=RCIM_Cubic, leaveTangent=6.5707, tangentMode=RCTM_User, time=0, value=0}
           {interpMode=RCIM_Cubic, tangentMode=RCTM_User, time=0.3, value=1}
           {tangentMode=RCTM_User, time=1, value=0}
          ]}}}
       Scale Curve=&RandomRangeFloat(Minimum=0.1, Maximum=0.5)
      )
      B=&MakeFloatFromLinearColor(LinearColor=$System.TODColorMultiplier, Channel:ENiagaraLinearColor_Channels=::NewEnumerator3)
     )
    )
    Scale RGB=&Multiply_Vector(
     A=&VectorFromFloat(Value=&Multiply_Float(
       A=&RandomRangeFloat(Minimum=0.75, Maximum=1)
       B=&SelectFloatFromArray(Direct Array Index=$Engine.QualityLevel, Float Selection Array=DataInterface() {propertyValues {FloatData [10,5,1,1,1]}})
      ))
     B=&MakeVectorFromLinearColorRGB(Color=$System.TODColorMultiplier)
    )
    ScaleA=true
    ScaleRGB=true
   ) {module=ScaleColor001}
   &ScaleSpriteSize(
    Uniform Curve Scale=&MakeCustomFloatFromBool(
     Bool=$User.bIsFrontEndPreview
     True Float=1
     False Float=&FloatFromCurve(
      FloatCurve=DataInterface() {propertyValues {Curve {keys [<at 0 1>,<at 1000 2>]}}}
      CurveIndex=$Engine.Owner.LODDistance
      Scale Curve=&FNTimeOfDayScalar(Night Scale=1.25, Morning Scale=1, Evening Scale=1)
     )
    )
    Uniform Curve Sprite Scale=DataInterface() {propertyValues {Curve {keys [
        {interpMode=RCIM_Cubic, tangentMode=RCTM_User, time=0, value=0}
        {interpMode=RCIM_Cubic, tangentMode=RCTM_User, time=0.5, value=1}
        {tangentMode=RCTM_User, time=1, value=0}
       ]}}}
   )
   /SetVariables(
    Particles.CenterFlareColor:LinearColor=&MultiplyLinearColor(
     Linear Color=$Particles.Color
     Scale Factor=&MakeVector4_FromLinearColor(LinearColor=&LerpLinearColors(StartColor=<rgba 5 4.8775 3.0696 1>, EndColor=<rgba 3.0348 3.9019 5 1>, LerpFactor=$User.Variant))
    )
    Particles.CenterFlareSize:Vector2f=&Multiply_Vector2(A=$Particles.SpriteSize, B=&Vector2DFromFloat(Value=0.5))
    Particles.CoronaFlareColor:LinearColor=&MultiplyLinearColor(
     Linear Color=$Particles.Color
     Scale Factor=&MakeVector4_FromLinearColor(LinearColor=&LerpLinearColors(StartColor=<rgba 0.15 0.15 0.15 1>, EndColor=<rgba 0.0729 0.0729 0.0729 1>, LerpFactor=$User.Variant))
    )
    Particles.CoronaFlareSize:Vector2f=&Multiply_Vector2(A=$Particles.Initial.SpriteSize, B=&Vector2DFromFloat(Value=0.75))
   ) {isSetParameters=true, module=SetVariables_#5}
  ]
  rendererClasses [&NiagaraSpriteRendererProperties]
  renderers [
   {rendererClass=&NiagaraSpriteRendererProperties}
   {rendererClass=&NiagaraSpriteRendererProperties, rendererIndex=1}
  ]
 }
}
summary {
 systemName=NS_Guitar_DimeBlanketGrace_Idle
 emitters [
  {emitterName=Sparkles, rendererClasses [&NiagaraSpriteRendererProperties]}
  {bEnabled=false, emitterName=Debug, rendererClasses [&NiagaraSpriteRendererProperties]}
 ]
 userVariables [
  User.bIsFrontEnd {type=NiagaraBool}
  User.bIsFrontEndPreview {type=NiagaraBool}
  User.MusicReactive {type=NiagaraBool, defaultValue=true}
  User.Variant {type=NiagaraFloat, defaultValue=0, description=""}
 ]
}
